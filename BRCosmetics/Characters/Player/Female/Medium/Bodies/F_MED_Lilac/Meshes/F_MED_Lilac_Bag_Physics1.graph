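# Physics asset for the bag on the F_MED_Lilac female medium player body cosmetic, apparently for secondary motion.
# Capsules cover four dyn_bag bones, plus the pelvis and both thighs, likely as body colliders for the bag.
# The thighs and bag master have angularly locked constraints to the pelvis.
# Bag_01 hangs off the master with nonzero swing1 and twist limits, while bag_02 and bag_03 branch from bag_01 with zero-degree limits.
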PhysicsAsset /BRCosmetics/Characters/Player/Female/Medium/Bodies/F_MED_Lilac/Meshes/F_MED_Lilac_Bag_Physics1
bodies {
 dyn_bag_01 [{
   shapeType=Capsule
   length=16.505
   radius=2.505
   shapeName=dyn_bag_01_capsule
   center=<21.0932 1.568 0.9865>
   rotation=<rot -12.9062 -90.0003 -10.0528>
  }]
 dyn_bag_02 [{
   shapeType=Capsule
   length=12.9398
   radius=2.8808
   shapeName=dyn_bag_02_capsule
   center=<9.8125 0.582 -4.2377>
   rotation=<rot 0.3497 -90 27.4464>
  }]
 dyn_bag_03 [{
   shapeType=Capsule
   length=11.9452
   radius=3.3516
   shapeName=dyn_bag_03_capsule
   center=<6.2686 0.7841 4.9546>
   rotation=<rot -2.2354 -95.7564 -34.2582>
  }]
 dyn_bag_Master [{shapeType=Capsule, length=20, radius=4, shapeName=dyn_bag_Master_capsule}]
 pelvis [{
   shapeType=Capsule
   length=7.4287
   radius=12.7066
   shapeName=pelvis_capsule
   center=<-0.0242 -2.0065 0.117>
   rotation=<rot 3.3365 -90.6898 -90>
  }]
 thigh_l [{shapeType=Capsule, length=35, radius=10, center=<-26.4785 0.8714 0.0646>, rotation=<rot 2.3368 79.4512 -85.9403>}]
 thigh_r [{shapeType=Capsule, length=33.5, radius=10.6165, center=<26.65 -0.5109 -0.8524>, rotation=<rot 1.7938 80.4925 -84.0692>}]
}
constraints [
 {
  bone1Name=thigh_l
  bone2Name=pelvis
  swing1LimitDegrees=45
  swing1Motion=Locked
  swing2LimitDegrees=45
  swing2Motion=Locked
  twistLimitDegrees=45
  twistMotion=Locked
 }
 {
  bone1Name=thigh_r
  bone2Name=pelvis
  swing1LimitDegrees=45
  swing1Motion=Locked
  swing2LimitDegrees=45
  swing2Motion=Locked
  twistLimitDegrees=45
  twistMotion=Locked
 }
 {
  bone1Name=dyn_bag_Master
  bone2Name=pelvis
  swing1LimitDegrees=45
  swing1Motion=Locked
  swing2LimitDegrees=45
  swing2Motion=Locked
  twistLimitDegrees=45
  twistMotion=Locked
 }
 {bone1Name=dyn_bag_01, bone2Name=dyn_bag_Master, swing1LimitDegrees=30, swing2LimitDegrees=0, twistLimitDegrees=30}
 {bone1Name=dyn_bag_02, bone2Name=dyn_bag_01, swing1LimitDegrees=0, swing2LimitDegrees=0, twistLimitDegrees=0}
 {bone1Name=dyn_bag_03, bone2Name=dyn_bag_01, swing1LimitDegrees=0, swing2LimitDegrees=0, twistLimitDegrees=0}
]
