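# Physics asset for the F_MED_Lilac female character's Belt A mesh, defining the bodies and joints for a dangling belt chain.
# Six capsule bones, dyn_belt_A_01 to 06, chain from a MASTER body locked to the pelvis, with a wide twist-locked swing at the root and generally tighter limits toward the tip.
# Capsules on pelvis, spine_01, right thigh and right forearm, joined by angularly locked constraints, supply body shapes around the belt.
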PhysicsAsset /BRCosmetics/Characters/Player/Female/Medium/Bodies/F_MED_Lilac/Meshes/F_MED_Lilac_BeltA_Physics
bodies {
 dyn_belt_A_01 [{shapeType=Capsule, length=0, radius=1.3, shapeName=dyn_belt_A_01_capsule, center=<5.5599 -0.422 0>, rotation=<rot -90 -179.9982 179.998>}]
 dyn_belt_A_02 [{shapeType=Capsule, length=2, radius=2, shapeName=dyn_belt_A_02_capsule, center=<4.9676 0 0>, rotation=<rot 90 0 0>}]
 dyn_belt_A_03 [{shapeType=Capsule, length=3.25, radius=1.8, shapeName=dyn_belt_A_03_capsule, center=<4.4355 0 0>, rotation=<rot 90 0 0>}]
 dyn_belt_A_04 [{shapeType=Capsule, length=2, radius=1.8, shapeName=dyn_belt_A_04_capsule, center=<3.3109 0 0>, rotation=<rot 90 0 0>}]
 dyn_belt_A_05 [{shapeType=Capsule, length=4.75, radius=1.8, shapeName=dyn_belt_A_05_capsule, center=<4.4812 0 0>, rotation=<rot 90 0 0>}]
 dyn_belt_A_06 [{shapeType=Capsule, length=1.25, radius=2.55, shapeName=dyn_belt_A_06_capsule, center=<5.7384 0 0>, rotation=<rot 90 0 0>}]
 dyn_belt_A_MASTER [{shapeType=Capsule, length=0.505, radius=0.505, shapeName=dyn_belt_A_MASTER_capsule, rotation=<rot 0 -90 0>}]
 lowerarm_r [{shapeType=Capsule, length=13.25, radius=5.75, shapeName=lowerarm_r_capsule, center=<-10.7176 0 0>, rotation=<rot -90 90 -90>}]
 pelvis [{shapeType=Capsule, length=14.4287, radius=11.2066, shapeName=pelvis_capsule, center=<-3.668 -4.0041 0>}]
 spine_01 [{shapeType=Capsule, length=10, radius=10, shapeName=spine_01_capsule, center=<6.9764 1.1496 0>, rotation=<rot 0 -90.0002 90>}]
 thigh_r [{shapeType=Capsule, length=44.6636, radius=11.8665, center=<21.8892 2.9153 -2.1069>, rotation=<rot 76.5626 178.2577 0.2982>}]
}
constraints [
 {
  bone1Name=thigh_r
  bone2Name=pelvis
  swing1LimitDegrees=45
  swing1Motion=Locked
  swing2LimitDegrees=45
  swing2Motion=Locked
  twistLimitDegrees=45
  twistMotion=Locked
 }
 {bone1Name=dyn_belt_A_06, bone2Name=dyn_belt_A_05, swing1LimitDegrees=20, swing2LimitDegrees=15, twistLimitDegrees=5}
 {bone1Name=dyn_belt_A_05, bone2Name=dyn_belt_A_04, swing1LimitDegrees=20, swing2LimitDegrees=15, twistLimitDegrees=5}
 {bone1Name=dyn_belt_A_04, bone2Name=dyn_belt_A_03, swing1LimitDegrees=20, swing2LimitDegrees=20, twistLimitDegrees=5}
 {bone1Name=dyn_belt_A_03, bone2Name=dyn_belt_A_02, swing1LimitDegrees=20, swing2LimitDegrees=30, twistLimitDegrees=20}
 {bone1Name=dyn_belt_A_02, bone2Name=dyn_belt_A_01, swing1LimitDegrees=20, swing2LimitDegrees=30, twistLimitDegrees=20}
 {bone1Name=dyn_belt_A_01, bone2Name=dyn_belt_A_MASTER, swing1LimitDegrees=70, swing2LimitDegrees=40, twistLimitDegrees=0, twistMotion=Locked}
 {
  bone1Name=dyn_belt_A_MASTER
  bone2Name=pelvis
  swing1LimitDegrees=45
  swing1Motion=Locked
  swing2LimitDegrees=45
  swing2Motion=Locked
  twistLimitDegrees=45
  twistMotion=Locked
 }
 {
  bone1Name=lowerarm_r
  bone2Name=spine_01
  swing1LimitDegrees=45
  swing1Motion=Locked
  swing2LimitDegrees=45
  swing2Motion=Locked
  twistLimitDegrees=45
  twistMotion=Locked
 }
 {
  bone1Name=spine_01
  bone2Name=pelvis
  swing1LimitDegrees=45
  swing1Motion=Locked
  swing2LimitDegrees=45
  swing2Motion=Locked
  twistLimitDegrees=45
  twistMotion=Locked
 }
]
