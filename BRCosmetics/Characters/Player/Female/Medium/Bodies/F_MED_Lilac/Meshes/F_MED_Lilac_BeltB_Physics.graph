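# Physics asset for the dangling belt B on the F_MED_Lilac female player body, built from capsule bodies on the belt bones, pelvis and left thigh.
# A master bone locked to the pelvis roots two chains with limited swing and locked twist, one through a second locked master.
# The chain ends (04 and 09) are constrained together, closing the belt into a loop.
# The left thigh is also locked to the pelvis.
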PhysicsAsset /BRCosmetics/Characters/Player/Female/Medium/Bodies/F_MED_Lilac/Meshes/F_MED_Lilac_BeltB_Physics
bodies {
 dyn_belt_B_01 [{shapeType=Capsule, length=1, radius=0.8, shapeName=dyn_belt_B_01_capsule, center=<2.8859 0 0>, rotation=<rot 90 0 0>}]
 dyn_belt_B_01_MASTER [{shapeType=Capsule, length=1, radius=2, shapeName=dyn_belt_B_01_MASTER_capsule}]
 dyn_belt_B_01_low [{shapeType=Capsule, length=2.25, radius=1.55, shapeName=dyn_belt_B_01_low_capsule, center=<2.4821 -1.0332 0>, rotation=<rot 90 0 0>}]
 dyn_belt_B_02 [{shapeType=Capsule, length=1.5, radius=1.2, shapeName=dyn_belt_B_02_capsule, center=<2.4821 -0.9127 0>, rotation=<rot 90 0 0>}]
 dyn_belt_B_03 [{shapeType=Capsule, length=1, radius=2, shapeName=dyn_belt_B_03_capsule, center=<2.4821 0 0>, rotation=<rot 90 0 0>}]
 dyn_belt_B_04 [{shapeType=Capsule, length=1, radius=2, shapeName=dyn_belt_B_04_capsule, center=<1.2795 0 0>, rotation=<rot 90 0 0>}]
 dyn_belt_B_05 [{shapeType=Capsule, length=1, radius=2, shapeName=dyn_belt_B_05_capsule, center=<2.463 0 0>, rotation=<rot 90 0 0>}]
 dyn_belt_B_06 [{shapeType=Capsule, length=1, radius=2, shapeName=dyn_belt_B_06_capsule, center=<2.463 0 0>, rotation=<rot 90 0 0>}]
 dyn_belt_B_07 [{shapeType=Capsule, length=1, radius=2, shapeName=dyn_belt_B_07_capsule, center=<2.463 0 0>, rotation=<rot 90 0 0>}]
 dyn_belt_B_08 [{shapeType=Capsule, length=1, radius=2, shapeName=dyn_belt_B_08_capsule, center=<2.463 0 0>, rotation=<rot 90 0 0>}]
 dyn_belt_B_09 [{shapeType=Capsule, length=1, radius=2, shapeName=dyn_belt_B_09_capsule, rotation=<rot 90 0 0>}]
 dyn_belt_B_MASTER [{shapeType=Capsule, length=1, radius=2, shapeName=dyn_belt_B_MASTER_capsule}]
 pelvis [{shapeType=Capsule, length=12.75, radius=10.5, shapeName=pelvis_capsule, center=<-6.5537 -3.9688 -6.0403>, rotation=<rot 90 0 0>}]
 thigh_l [{
   shapeType=Capsule
   length=59.75
   radius=10.25
   shapeName=thigh_l_capsule
   center=<-28.0871 -2.6864 -3.1649>
   rotation=<rot 75.9375 179.9995 179.9999>
  }]
}
constraints [
 {bone1Name=dyn_belt_B_05, bone2Name=dyn_belt_B_MASTER, swing1LimitDegrees=25, swing2LimitDegrees=25, twistLimitDegrees=1, twistMotion=Locked}
 {bone1Name=dyn_belt_B_06, bone2Name=dyn_belt_B_05, swing1LimitDegrees=45, swing2LimitDegrees=5, twistLimitDegrees=1, twistMotion=Locked}
 {bone1Name=dyn_belt_B_07, bone2Name=dyn_belt_B_06, swing1LimitDegrees=45, swing2LimitDegrees=25, twistLimitDegrees=1, twistMotion=Locked}
 {bone1Name=dyn_belt_B_08, bone2Name=dyn_belt_B_07, swing1LimitDegrees=45, swing2LimitDegrees=25, twistLimitDegrees=1, twistMotion=Locked}
 {bone1Name=dyn_belt_B_09, bone2Name=dyn_belt_B_08, swing1LimitDegrees=45, swing2LimitDegrees=25, twistLimitDegrees=1, twistMotion=Locked}
 {
  bone1Name=dyn_belt_B_01_MASTER
  bone2Name=dyn_belt_B_MASTER
  swing1LimitDegrees=45
  swing1Motion=Locked
  swing2LimitDegrees=45
  swing2Motion=Locked
  twistLimitDegrees=1
  twistMotion=Locked
 }
 {
  bone1Name=dyn_belt_B_01
  bone2Name=dyn_belt_B_01_MASTER
  swing1LimitDegrees=50
  swing2LimitDegrees=70
  twistLimitDegrees=1
  twistMotion=Locked
 }
 {bone1Name=dyn_belt_B_01_low, bone2Name=dyn_belt_B_01, swing1LimitDegrees=60, swing2LimitDegrees=5, twistLimitDegrees=1, twistMotion=Locked}
 {bone1Name=dyn_belt_B_02, bone2Name=dyn_belt_B_01_low, swing1LimitDegrees=45, swing2LimitDegrees=1, twistLimitDegrees=1, twistMotion=Locked}
 {bone1Name=dyn_belt_B_03, bone2Name=dyn_belt_B_02, swing1LimitDegrees=45, swing2LimitDegrees=25, twistLimitDegrees=1, twistMotion=Locked}
 {bone1Name=dyn_belt_B_04, bone2Name=dyn_belt_B_03, swing1LimitDegrees=45, swing2LimitDegrees=25, twistLimitDegrees=1, twistMotion=Locked}
 {
  bone1Name=dyn_belt_B_MASTER
  bone2Name=pelvis
  swing1LimitDegrees=45
  swing1Motion=Locked
  swing2LimitDegrees=45
  swing2Motion=Locked
  twistLimitDegrees=1
  twistMotion=Locked
 }
 {
  bone1Name=thigh_l
  bone2Name=pelvis
  swing1LimitDegrees=45
  swing1Motion=Locked
  swing2LimitDegrees=45
  swing2Motion=Locked
  twistLimitDegrees=45
  twistMotion=Locked
 }
 {bone1Name=dyn_belt_B_04, bone2Name=dyn_belt_B_09, swing1LimitDegrees=45, swing2LimitDegrees=25, twistLimitDegrees=1, twistMotion=Locked}
]
