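# Post-process style animation Blueprint for the F_MED_Lilac body cosmetic, adding secondary motion to its bag, belts and zippers.
# The parent pose runs through Control Rigs, then cached bag and belt stages pick RigidBody simulation or bone-constraint fallbacks by movement and front-end state.
# Constrained, AnimDynamics-driven zippers finish the pose.
# The EventGraph derives state alphas and gravity, feeds material WPO cloth and wind, and resets dynamics around skydiving.
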AnimBlueprint /BRCosmetics/Characters/Player/Female/Medium/Bodies/F_MED_Lilac/Meshes/F_MED_Lilac_AnimBP
target_skeleton=&/F_MED_Lilac_Skeleton
graphs [
 AnimGraph {nodes [
   n1 = Get isCrouchingMovingSprinting() -> (isCrouchingMovingSprinting:double)
   n2 = "Transform (Modify) Bone - Bone: dyn_bag_02"!(ComponentPose, Alpha=@n1, Translation="0,-4,0", Rotation="0,0,0") -> (Pose) {
    rotationMode=BMM_Additive
    rotationSpace=BCS_BoneSpace
    translationMode=BMM_Additive
    translationSpace=BCS_BoneSpace
    alphaBoolBlend {blendOption=Linear}
    alphaScaleBiasClamp {bInterpResult=true, interpSpeedDecreasing=30, interpSpeedIncreasing=30}
    boneToModify {boneName=dyn_bag_02}
    translation {y=-4}
   }
   "Transform (Modify) Bone - Bone: dyn_bag_03"!(ComponentPose=@n2, Alpha=@n1, Translation="0,-6,1", Rotation="0,0,0") -> (Pose) {
    rotationMode=BMM_Replace
    rotationSpace=BCS_BoneSpace
    translationMode=BMM_Additive
    translationSpace=BCS_BoneSpace
    alphaBoolBlend {blendOption=Linear}
    alphaScaleBiasClamp {bInterpResult=true, interpSpeedDecreasing=30, interpSpeedIncreasing=30}
    boneToModify {boneName=dyn_bag_03}
    translation {y=-6, z=1}
   }
   "Transform (Modify) Bone - Bone: dyn_bag_01"!(ComponentPose, bAlphaBoolEnabled:bool=Get bIsCrouching(), Translation="0,0,0") -> (Pose) {
    alphaInputType=Bool
    rotationMode=BMM_Additive
    rotationSpace=BCS_BoneSpace
    translationSpace=BCS_BoneSpace
    alphaBoolBlend {blendOption=Linear}
    alphaScaleBiasClamp {bInterpResult=true, interpSpeedDecreasing=30, interpSpeedIncreasing=30}
    boneToModify {boneName=dyn_bag_01}
    rotation {roll=10}
   }
   n3 = Get bIsCrouching()
   n4 = "Transform (Modify) Bone - Bone: dyn_bag_02"!(ComponentPose, bAlphaBoolEnabled:bool=@n3, Translation="0,6,0", Rotation="0,0,5") -> (Pose) {
    alphaInputType=Bool
    rotationMode=BMM_Additive
    rotationSpace=BCS_BoneSpace
    translationMode=BMM_Additive
    translationSpace=BCS_BoneSpace
    alphaBoolBlend {blendOption=Linear}
    alphaScaleBiasClamp {bInterpResult=true, interpSpeedDecreasing=30, interpSpeedIncreasing=30}
    boneToModify {boneName=dyn_bag_02}
    rotation {roll=5}
    translation {y=6}
   }
   "Transform (Modify) Bone - Bone: dyn_bag_03"!(ComponentPose=@n4, bAlphaBoolEnabled:bool=@n3, Translation="0,6,2", Rotation="0,0,-15") -> (Pose) {
    alphaInputType=Bool
    rotationMode=BMM_Additive
    rotationSpace=BCS_BoneSpace
    translationMode=BMM_Additive
    translationSpace=BCS_BoneSpace
    alphaBoolBlend {blendOption=Linear}
    alphaScaleBiasClamp {bInterpResult=true, interpSpeedDecreasing=30, interpSpeedIncreasing=30}
    boneToModify {boneName=dyn_bag_03}
    rotation {roll=-15}
    translation {y=6, z=2}
   }
   n5 = "Constraint - dyn_L_zipper_MASTER (PPPPPPPPPPPPPPPP)"!(
    ComponentPose=Local To Component(LocalPose=Use cached pose 'BeltB'())
    Alpha=1
    ConstraintWeights_0=0.6813
    ConstraintWeights_1:float=0.2276
    ConstraintWeights_2:float=0.0523
    ConstraintWeights_3:float=0.0142
    ConstraintWeights_4:float=0.0124
    ConstraintWeights_5:float=0.012
    ConstraintWeights_6:float=0.0002
    ConstraintWeights_7:float=0
    ConstraintWeights_8:float=0
    ConstraintWeights_9:float=0
    ConstraintWeights_10:float=0
    ConstraintWeights_11:float=0
    ConstraintWeights_12:float=0
    ConstraintWeights_13:float=0
    ConstraintWeights_14:float=0
    ConstraintWeights_15:float=0
   ) -> (Pose) {
    alphaBoolBlend {blendOption=Linear}
    boneToModify {boneName=dyn_L_zipper_MASTER}
    constraintSetup [
     {targetBone {boneName=thigh_l}}
     {targetBone {boneName=thigh_twist_01_l}}
     {targetBone {boneName=deform_groin_l}}
     {targetBone {boneName=dyn_bag_04}}
     {targetBone {boneName=thigh_twist_01_r}}
     {targetBone {boneName=thigh_r}}
     {targetBone {boneName=deform_groin_r}}
     {targetBone {boneName=calf_l}}
     {targetBone {boneName=deform_knee_in_l}}
     {targetBone {boneName=deform_knee_out_base_l}}
     {targetBone {boneName=dyn_belt_B_05}}
     {targetBone {boneName=deform_knee_in_r}}
     {targetBone {boneName=calf_twist_02_l}}
     {targetBone {boneName=calf_r}}
     {targetBone {boneName=deform_knee_out_base_r}}
     {targetBone {boneName=calf_twist_02_r}}
    ]
    constraintWeights [1,1,1,1,1,1,1,1,1,1,1,1,1,1,1,1]
   }
   n6 = "Constraint - dyn_R_zipper_MASTER (PPPPPPPPPPPPPPPP)"!(
    ComponentPose=@n5
    Alpha=1
    ConstraintWeights_0=0.6798
    ConstraintWeights_1:float=0.2271
    ConstraintWeights_2:float=0.0533
    ConstraintWeights_3:float=0.0143
    ConstraintWeights_4:float=0.0128
    ConstraintWeights_5:float=0.0125
    ConstraintWeights_6:float=0.0002
    ConstraintWeights_7:float=0
    ConstraintWeights_8:float=0
    ConstraintWeights_9:float=0
    ConstraintWeights_10:float=0
    ConstraintWeights_11:float=0
    ConstraintWeights_12:float=0
    ConstraintWeights_13:float=0
    ConstraintWeights_14:float=0
    ConstraintWeights_15:float=0
   ) -> (Pose) {
    alphaBoolBlend {blendOption=Linear}
    boneToModify {boneName=dyn_R_zipper_MASTER}
    constraintSetup [
     {targetBone {boneName=thigh_r}}
     {targetBone {boneName=thigh_twist_01_r}}
     {targetBone {boneName=deform_groin_r}}
     {targetBone {boneName=dyn_bag_04}}
     {targetBone {boneName=thigh_twist_01_l}}
     {targetBone {boneName=thigh_l}}
     {targetBone {boneName=deform_groin_l}}
     {targetBone {boneName=calf_r}}
     {targetBone {boneName=deform_knee_in_r}}
     {targetBone {boneName=deform_knee_out_base_r}}
     {targetBone {boneName=deform_knee_in_l}}
     {targetBone {boneName=dyn_belt_B_05}}
     {targetBone {boneName=calf_l}}
     {targetBone {boneName=calf_twist_02_r}}
     {targetBone {boneName=deform_knee_out_base_l}}
     {targetBone {boneName=calf_twist_02_l}}
    ]
    constraintWeights [1,1,1,1,1,1,1,1,1,1,1,1,1,1,1,1]
   }
   n7 = Anim Dynamics - Bone: dyn_R_zipper_MASTER(ComponentPose=@n6, Alpha=1, ExternalForce="0,0,0") {
    angularDampingOverride=1
    angularSpringConstant=20000
    bAngularSpring=true
    bGravityOverrideInSimSpace=true
    bOverrideAngularDamping=true
    bUseGravityOverride=true
    gravityScale=0.35
    simulationSpace=RootRelative
    alphaBoolBlend {blendOption=Linear}
    boundBone {boneName=dyn_R_zipper_MASTER}
    componentLinearAccScale=<1 1 1>
    componentLinearVelScale=<1 1 1>
    gravityOverride {z=-980}
    physicsBodyDefinitions [{
      sphereCollisionRadius=0
      boundBone {boneName=dyn_R_zipper_MASTER}
      boxExtents=<20 30 30>
      constraintSetup {angularLimitsMax=<5 5 5>, angularLimitsMin=<-5 -5 -80>, angularTarget=<1 0 0>}
      localJointOffset=<4 0 0>
     }]
   }
   n8 = Anim Dynamics - Bone: dyn_L_zipper_MASTER(ComponentPose=@n7, Alpha=1, ExternalForce="0,0,0") {
    angularDampingOverride=1
    angularSpringConstant=20000
    bAngularSpring=true
    bGravityOverrideInSimSpace=true
    bOverrideAngularDamping=true
    bUseGravityOverride=true
    gravityScale=0.35
    simulationSpace=RootRelative
    alphaBoolBlend {blendOption=Linear}
    boundBone {boneName=dyn_L_zipper_MASTER}
    componentLinearAccScale=<1 1 1>
    componentLinearVelScale=<1 1 1>
    gravityOverride {z=-980}
    physicsBodyDefinitions [{
      sphereCollisionRadius=0
      boundBone {boneName=dyn_L_zipper_MASTER}
      boxExtents=<20 30 30>
      constraintSetup {angularLimitsMax=<5 5 5>, angularLimitsMin=<-5 -5 -80>, angularTarget=<1 0 0>}
      localJointOffset=<4 0 0>
     }]
   }
   Output Pose(Result=Component To Local(ComponentPose=@n8)) {layerGroup=None}
   n9 = "Transform (Modify) Bone - Bone: dyn_bag_Master"!(
    ComponentPose=Local To Component(LocalPose=Use cached pose 'Start'())
    Alpha=Get IsCrouchOrSkydivingOrDBNO() -> (IsCrouchOrSkydivingOrDBNO:double)
    Translation="1.6,0,0"
    Rotation="0,-20,0"
   ) -> (Pose) {
    rotationMode=BMM_Additive
    rotationSpace=BCS_BoneSpace
    translationMode=BMM_Additive
    translationSpace=BCS_BoneSpace
    alphaBoolBlend {blendOption=Linear}
    alphaScaleBiasClamp {bInterpResult=true, interpSpeedDecreasing=30, interpSpeedIncreasing=30}
    boneToModify {boneName=dyn_bag_Master}
    rotation {yaw=-20}
    translation {x=1.6}
   }
   n10 = RigidBody!(ComponentPose=@n9, Alpha, OverrideWorldGravity:Vector="0,0,-980", ExternalForce) -> (Pose) {
    bOverrideWorldGravity=true
    overridePhysicsAsset=&/F_MED_Lilac_Bag_Physics1
    simulationSpace=BaseBoneSpace
    alphaBoolBlend {blendOption=Linear}
    baseBoneRef {boneName=pelvis}
    componentAppliedLinearAccClamp=<1500 1500 1500>
    componentLinearAccScale=<0.3 0.3 0.4>
    componentLinearVelScale=<0.2 0.2 0.3>
   }
   n11 = Get IsNotCrouching() -> (IsNotCrouching:double)
   n12 = "Constraint - dyn_bag_01 (RRR)"!(
    ComponentPose=Local To Component(LocalPose=Use cached pose 'Start'())
    Alpha=@n11
    ConstraintWeights_0=0.5
    ConstraintWeights_1:float=0.4
    ConstraintWeights_2:float=0.1
   ) -> (Pose) {
    alphaBoolBlend {blendOption=Linear}
    boneToModify {boneName=dyn_bag_01}
    constraintSetup [
     {transformType=Rotation, targetBone {boneName=thigh_l}}
     {transformType=Rotation, targetBone {boneName=thigh_r}}
     {transformType=Rotation, targetBone {boneName=pelvis}}
    ]
    constraintWeights [1,1,1]
   }
   n13 = "Constraint - dyn_bag_02 (PP)"!(ComponentPose=@n12, Alpha=@n11, ConstraintWeights_0=0.3, ConstraintWeights_1:float=0.7) -> (Pose) {
    alphaBoolBlend {blendOption=Linear}
    boneToModify {boneName=dyn_bag_02}
    constraintSetup [{targetBone {boneName=thigh_l}}, {targetBone {boneName=dyn_bag_01}}]
    constraintWeights [1,1]
   }
   n14 = "Transform (Modify) Bone - Bone: dyn_bag_02"!(ComponentPose=@n13, Alpha:float=1, Translation="0.394,-1.786,1.844", Rotation="0,0,0") -> (Pose) {
    rotationMode=BMM_Replace
    rotationSpace=BCS_BoneSpace
    translationMode=BMM_Additive
    translationSpace=BCS_BoneSpace
    alphaBoolBlend {blendOption=Linear}
    alphaScaleBiasClamp {bInterpResult=true, interpSpeedDecreasing=30, interpSpeedIncreasing=30}
    boneToModify {boneName=dyn_bag_02}
    translation=<0.394 -1.786 1.844>
   }
   n15 = "Transform (Modify) Bone - Bone: dyn_bag_03"!(ComponentPose=@n14, Alpha:float=1, Translation="0.000,-0.229,1.465", Rotation="0,0,0") -> (Pose) {
    rotationMode=BMM_Replace
    rotationSpace=BCS_BoneSpace
    translationMode=BMM_Additive
    translationSpace=BCS_BoneSpace
    alphaBoolBlend {blendOption=Linear}
    alphaScaleBiasClamp {bInterpResult=true, interpSpeedDecreasing=30, interpSpeedIncreasing=30}
    boneToModify {boneName=dyn_bag_03}
    translation {y=-0.229, z=1.465}
   }
   n16 = "Constraint - dyn_bag_01 (RRR)"!(ComponentPose=@n15, bAlphaBoolEnabled:bool=Get bIsCrouching(), ConstraintWeights_0=0.4, ConstraintWeights_1=0.4, ConstraintWeights_2=0.2) -> (Pose) {
    alphaInputType=Bool
    alphaBoolBlend {blendOption=Linear}
    boneToModify {boneName=dyn_bag_01}
    constraintSetup [
     {transformType=Rotation, targetBone {boneName=thigh_l}}
     {transformType=Rotation, targetBone {boneName=thigh_r}}
     {transformType=Rotation, targetBone {boneName=pelvis}}
    ]
    constraintWeights [1,1,1]
   }
   n17 = Blend Poses by bool(
    BlendPose_0=Component To Local(ComponentPose=@n10)
    BlendPose_1=Component To Local(ComponentPose=@n16)
    bActiveValue=Get RgidbodyNOTFE() -> (RgidbodyNOTFE:bool)
   ) {blendType=Linear}
   Save cached pose 'BagCache'(Pose=Component To Local(ComponentPose=Local To Component(LocalPose=@n17))) {cacheName=BagCache}
   n18 = Control Rig!(
    Source=Copy Pose From Mesh() {bUseAttachedParent=true}
    bAlphaBoolEnabled:bool=True
    isCrouching:bool=Get bIsCrouching()
    IsEmote:bool=Get bIsPlayingEmote()
   ) -> (Pose) {
    alphaInputType=Bool
    alphaBoolBlend {blendOption=Linear}
    controlRigAssetReference {blueprintRigClass=&/F_MED_Lilac_BeltB_CtrlRig.F_MED_Lilac_BeltB_CtrlRig_C}
   }
   n19 = Control Rig!(Source=@n18, Alpha:float) -> (Pose) {
    alphaBoolBlend {blendOption=Linear}
    controlRigAssetReference {blueprintRigClass=&/F_MED_DeformRig_Lilac_ControlRig.F_MED_DeformRig_Lilac_ControlRig_C}
   }
   Save cached pose 'Start'(Pose=@n19) {cacheName=Start}
   n20 = "Constraint - dyn_belt_A_01 (P)"(ComponentPose=Local To Component(LocalPose=Use cached pose 'BagCache'()), Alpha=1, ConstraintWeights_0=1) {
    alphaBoolBlend {blendOption=Linear}
    boneToModify {boneName=dyn_belt_A_01}
    constraintSetup [{targetBone {boneName=thigh_r}}]
    constraintWeights [1]
   }
   n21 = "Constraint - dyn_belt_A_01 (PPPPPPPPPPPPPPPPPPP)"!(
    ComponentPose=Local To Component(LocalPose=Use cached pose 'BagCache'())
    ConstraintWeights_0=0.8402
    ConstraintWeights_1:float=0.1286
    ConstraintWeights_2:float=0.0205
    ConstraintWeights_3:float=0.007
    ConstraintWeights_4:float=0.0023
    ConstraintWeights_5:float=0.0009
    ConstraintWeights_6:float=0.0003
    ConstraintWeights_7:float=0.0001
    ConstraintWeights_8:float=0
    ConstraintWeights_9:float=0
    ConstraintWeights_10:float=0
    ConstraintWeights_11:float=0
    ConstraintWeights_12:float=0
    ConstraintWeights_13:float=0
    ConstraintWeights_14:float=0
    ConstraintWeights_15:float=0
    ConstraintWeights_16:float=0
    ConstraintWeights_17:float=0
    ConstraintWeights_18:float=0
   ) -> (Pose) {
    alphaBoolBlend {blendOption=Linear}
    boneToModify {boneName=dyn_belt_A_01}
    constraintSetup [
     {targetBone {boneName=pelvis}}
     {targetBone {boneName=deform_groin_r}}
     {targetBone {boneName=thigh_twist_01_r}}
     {targetBone {boneName=thigh_r}}
     {targetBone {boneName=thigh_twist_01_l}}
     {targetBone {boneName=thigh_l}}
     {targetBone {boneName=deform_groin_l}}
     {targetBone {boneName=deform_glute_r}}
     {targetBone {boneName=spine_02}}
     {targetBone {boneName=spine_01}}
     {targetBone {boneName=deform_knee_in_r}}
     {targetBone {boneName=deform_glute_l}}
     {targetBone {boneName=calf_r}}
     {targetBone {boneName=calf_twist_02_r}}
     {targetBone {boneName=deform_knee_out_r}}
     {targetBone {boneName=deform_knee_in_l}}
     {targetBone {boneName=calf_l}}
     {targetBone {boneName=calf_twist_02_l}}
     {targetBone {boneName=deform_knee_out_l}}
    ]
    constraintWeights [0.8402,0.1286,0.0205,0.007,0.0023,0.0009,0.0003,0.0001,0,0,0,0,0,0,0,0,0,0,0]
   }
   n22 = "Transform (Modify) Bone - Bone: dyn_belt_A_MASTER"!(ComponentPose=@n21, Alpha=Get Is Crouch Sprinting() -> (Is Crouch Sprinting:double), Translation="-3,6,0", Rotation="120,60,0") -> (Pose) {
    rotationMode=BMM_Replace
    rotationSpace=BCS_BoneSpace
    translationMode=BMM_Additive
    translationSpace=BCS_BoneSpace
    alphaBoolBlend {blendOption=Linear}
    alphaScaleBiasClamp {bInterpResult=true, interpSpeedDecreasing=5, interpSpeedIncreasing=7}
    boneToModify {boneName=dyn_belt_A_MASTER}
    rotation {pitch=120, yaw=60}
    translation=<-3 6>
   }
   n23 = "Transform (Modify) Bone - Bone: dyn_belt_A_MASTER"!(ComponentPose=@n22, bAlphaBoolEnabled:bool=Get bIsDBNO(), Translation="-10,0,0", Rotation="0,0,0") -> (Pose) {
    alphaInputType=Bool
    rotationMode=BMM_Replace
    rotationSpace=BCS_BoneSpace
    translationMode=BMM_Additive
    translationSpace=BCS_BoneSpace
    alphaBoolBlend {blendOption=Linear}
    alphaScaleBiasClamp {bInterpResult=true, interpSpeedDecreasing=5, interpSpeedIncreasing=7}
    boneToModify {boneName=dyn_belt_A_MASTER}
    translation {x=-10}
   }
   n24 = RigidBody(ComponentPose=@n23, OverrideWorldGravity=Get GravityOverride()) {
    bClampLinearTranslationLimitToRefPose=true
    bOverrideWorldGravity=true
    overridePhysicsAsset=&/F_MED_Lilac_BeltA_Physics
    simulationSpace=BaseBoneSpace
    alphaBoolBlend {blendOption=Linear}
    baseBoneRef {boneName=pelvis}
    componentAppliedLinearAccClamp=<3000 3000 3000>
    componentLinearAccScale=<0.3 0.2 0.5>
    componentLinearVelScale=<0.3 0.22 0.5>
   }
   Save cached pose 'BeltA'(Pose=Blend Poses by bool(BlendPose_0=Component To Local(ComponentPose=@n24), BlendPose_1=Component To Local(ComponentPose=@n20), bActiveValue=Get bRigidBodyEnabled()) {blendType=Linear}) {cacheName=BeltA}
   n25 = Get GravityOverride()
   n26 = RigidBody(ComponentPose=Local To Component(LocalPose=Use cached pose 'BeltA'()), OverrideWorldGravity=@n25) {
    bOverrideWorldGravity=true
    overridePhysicsAsset=&/F_MED_Lilac_BeltB_Physics
    simulationSpace=BaseBoneSpace
    alphaBoolBlend {blendOption=Linear}
    baseBoneRef {boneName=pelvis}
    componentAppliedLinearAccClamp=<1500 1500 1500>
    componentLinearAccScale=<0.3 0.3 0.4>
    componentLinearVelScale=<0.2 0.2 0.3>
   }
   n27 = "Transform (Modify) Bone - Bone: dyn_belt_B_03"!(ComponentPose=@n26, bAlphaBoolEnabled:bool=Get bIsInFrontEnd(), Rotation="0,20,0") -> (Pose) {
    alphaInputType=Bool
    rotationMode=BMM_Additive
    alphaBoolBlend {blendOption=Linear}
    boneToModify {boneName=dyn_belt_B_03}
    rotation {yaw=20}
   }
   n28 = RigidBody(ComponentPose=Local To Component(LocalPose=Use cached pose 'BeltA'()), OverrideWorldGravity=@n25) {
    bOverrideWorldGravity=true
    overridePhysicsAsset=&/F_MED_Lilac_BeltBCrouch_Physics
    simulationSpace=BaseBoneSpace
    alphaBoolBlend {blendOption=Linear}
    baseBoneRef {boneName=pelvis}
    componentAppliedLinearAccClamp=<1500 1500 1500>
    componentLinearAccScale=<0.3 0.3 0.4>
    componentLinearVelScale=<0.2 0.2 0.3>
   }
   n29 = "Transform (Modify) Bone - Bone: dyn_belt_B_05"!(ComponentPose=@n28, Alpha=1, Translation="0,-0.5,0.5") -> (Pose) {translationMode=BMM_Additive, alphaBoolBlend {blendOption=Linear}, boneToModify {boneName=dyn_belt_B_05}, translation {y=-0.5, z=0.5}}
   n30 = Blend Poses by bool(
    BlendPose_0=Component To Local(ComponentPose=@n29)
    BlendPose_1=Component To Local(ComponentPose=@n27)
    bActiveValue=Get IsCrouchingDBNO() -> (IsCrouchingDBNO:bool)
   ) {blendType=Linear}
   n31 = "Constraint - dyn_belt_B_05 (RR)"!(ComponentPose=Local To Component(LocalPose=Use cached pose 'BeltA'()), Alpha=1, ConstraintWeights_0=0.1, ConstraintWeights_1:float=0.9) -> (Pose) {
    alphaBoolBlend {blendOption=Linear}
    boneToModify {boneName=dyn_belt_B_05}
    constraintSetup [{transformType=Rotation, targetBone {boneName=thigh_l}}, {transformType=Rotation, targetBone {boneName=pelvis}}]
    constraintWeights [1,1]
   }
   n32 = "Constraint - dyn_belt_B_06 (RR)"!(ComponentPose=@n31, Alpha=1, ConstraintWeights_0=0.2, ConstraintWeights_1:float=0.8) -> (Pose) {
    alphaBoolBlend {blendOption=Linear}
    boneToModify {boneName=dyn_belt_B_06}
    constraintSetup [{transformType=Rotation, targetBone {boneName=thigh_l}}, {transformType=Rotation, targetBone {boneName=pelvis}}]
    constraintWeights [1,1]
   }
   n33 = "Constraint - dyn_belt_B_07 (RR)"!(ComponentPose=@n32, Alpha=1, ConstraintWeights_0=0.7, ConstraintWeights_1:float=0.3) -> (Pose) {
    alphaBoolBlend {blendOption=Linear}
    boneToModify {boneName=dyn_belt_B_07}
    constraintSetup [{transformType=Rotation, targetBone {boneName=thigh_l}}, {transformType=Rotation, targetBone {boneName=pelvis}}]
    constraintWeights [1,1]
   }
   n34 = "Constraint - dyn_belt_B_01_MASTER (PP)"!(ComponentPose=@n33, Alpha=1, ConstraintWeights_0=0.95, ConstraintWeights_1:float=0.05) -> (Pose) {
    alphaBoolBlend {blendOption=Linear}
    boneToModify {boneName=dyn_belt_B_01_MASTER}
    constraintSetup [{targetBone {boneName=thigh_l}}, {targetBone {boneName=pelvis}}]
    constraintWeights [1,1]
   }
   n35 = "Constraint - dyn_belt_B_02 (RR)"!(ComponentPose=@n34, Alpha=0.5, ConstraintWeights_0=0.8, ConstraintWeights_1:float=0.2) -> (Pose) {
    alphaBoolBlend {blendOption=Linear}
    boneToModify {boneName=dyn_belt_B_02}
    constraintSetup [{transformType=Rotation, targetBone {boneName=thigh_l}}, {transformType=Rotation, targetBone {boneName=pelvis}}]
    constraintWeights [1,1]
   }
   n36 = Component To Local(ComponentPose=@n35)
   n37 = Get bIsTacticalSprinting()
   n38 = "Constraint - dyn_belt_B_05 (RR)"(ComponentPose=Local To Component(LocalPose=Use cached pose 'BeltA'()), Alpha=1, ConstraintWeights_0=0.1, ConstraintWeights_1=0.9) {
    alphaBoolBlend {blendOption=Linear}
    boneToModify {boneName=dyn_belt_B_05}
    constraintSetup [{transformType=Rotation, targetBone {boneName=thigh_l}}, {transformType=Rotation, targetBone {boneName=pelvis}}]
    constraintWeights [1,1]
   }
   n39 = "Constraint - dyn_belt_B_06 (RR)"(ComponentPose=@n38, Alpha=1, ConstraintWeights_0=0.2, ConstraintWeights_1=0.8) {
    alphaBoolBlend {blendOption=Linear}
    boneToModify {boneName=dyn_belt_B_06}
    constraintSetup [{transformType=Rotation, targetBone {boneName=thigh_l}}, {transformType=Rotation, targetBone {boneName=pelvis}}]
    constraintWeights [1,1]
   }
   n40 = "Constraint - dyn_belt_B_07 (RR)"(ComponentPose=@n39, Alpha=1, ConstraintWeights_0=0.25, ConstraintWeights_1=0.75) {
    alphaBoolBlend {blendOption=Linear}
    boneToModify {boneName=dyn_belt_B_07}
    constraintSetup [{transformType=Rotation, targetBone {boneName=thigh_l}}, {transformType=Rotation, targetBone {boneName=pelvis}}]
    constraintWeights [1,1]
   }
   n41 = "Constraint - dyn_belt_B_01_MASTER (PP)"(ComponentPose=@n40, Alpha=1, ConstraintWeights_0=0.9, ConstraintWeights_1=0.1) {
    alphaBoolBlend {blendOption=Linear}
    boneToModify {boneName=dyn_belt_B_01_MASTER}
    constraintSetup [{targetBone {boneName=thigh_l}}, {targetBone {boneName=pelvis}}]
    constraintWeights [1,1]
   }
   n42 = "Constraint - dyn_belt_B_01_low (RR)"!(ComponentPose=@n41, Alpha=1, ConstraintWeights_0=0.2, ConstraintWeights_1:float=0.8) -> (Pose) {
    alphaBoolBlend {blendOption=Linear}
    boneToModify {boneName=dyn_belt_B_01_low}
    constraintSetup [{transformType=Rotation, targetBone {boneName=thigh_l}}, {transformType=Rotation, targetBone {boneName=pelvis}}]
    constraintWeights [1,1]
   }
   n43 = "Constraint - dyn_belt_B_02 (RR)"(ComponentPose=@n42, Alpha=0.5, ConstraintWeights_0=0.8, ConstraintWeights_1=0.2) {
    alphaBoolBlend {blendOption=Linear}
    boneToModify {boneName=dyn_belt_B_02}
    constraintSetup [{transformType=Rotation, targetBone {boneName=thigh_l}}, {transformType=Rotation, targetBone {boneName=pelvis}}]
    constraintWeights [1,1]
   }
   n44 = Get bIsInFrontEnd()
   n45 = "Transform (Modify) Bone - Bone: dyn_belt_B_01"!(ComponentPose=@n43, bAlphaBoolEnabled:bool=@n44, Translation="-2.573,-0.724,-2.426", Rotation="5,5,5", Scale="1,1,1") -> (Pose) {
    alphaInputType=Bool
    rotationMode=BMM_Replace
    rotationSpace=BCS_BoneSpace
    translationMode=BMM_Replace
    translationSpace=BCS_BoneSpace
    alphaBoolBlend {blendOption=Linear}
    boneToModify {boneName=dyn_belt_B_01}
    rotation=<rot 5 5 5>
    translation=<-2.573 -0.724 -2.426>
   }
   n46 = "Transform (Modify) Bone - Bone: dyn_bag_01"!(ComponentPose=@n45, bAlphaBoolEnabled=@n44, Translation="0,0,0", Rotation:Rotator="15,-2,0", Scale:Vector="1,1,1") -> (Pose) {
    alphaInputType=Bool
    rotationMode=BMM_Replace
    rotationSpace=BCS_BoneSpace
    translationSpace=BCS_BoneSpace
    alphaBoolBlend {blendOption=Linear}
    boneToModify {boneName=dyn_bag_01}
    rotation {pitch=15, yaw=-2}
   }
   Save cached pose 'BeltB'(Pose=Blend Poses by bool(
     BlendPose_0=Blend Poses by bool(BlendPose_0=@n36, BlendPose_1=@n30, bActiveValue=@n37) {blendType=Linear}
     BlendPose_1=Component To Local(ComponentPose=@n46)
     bActiveValue=Get bRigidBodyEnabled()
    ) {blendType=Linear}) {cacheName=BeltB}
   Get GravityOverride()
  ]}
 EventGraph {nodes [
   Event BlueprintInitializeAnimation()
   > Set SkelMesh(SkelMesh=GetOwnerSkeletalMeshComponent())
   n1 = TryGetPawnOwner()
   > Is Valid(InputObject=@n1)
   > n2 = Cast To FortPlayerPawnAthena(Object=@n1)
   > n3 = Set PlayerPawn(PlayerPawn:Fort Player Pawn Athena Object Reference=@n2) -> (Output_Get:Fort Player Pawn Athena Object Reference)
   > Set Body_Mesh(Body_Mesh=GetSkeletalMeshForPartType(self=@n3))
   n4 = Event OnBeginSkydiving()
   ResetDynamics(execute=[@n4.then, @n24.then], InTeleportType=TeleportPhysics)
   > ResetDynamics(InTeleportType=ResetPhysics)
   n5 = Event BlueprintUpdateAnimation()
   > n6 = Sequence!(execute) -> (then_0, then_1, then_2:Exec)
   > Is Valid(InputObject=Get PlayerPawn() -> (PlayerPawn:Fort Player Pawn Athena Object Reference))
   > Set WPO Globals Function(
    Material=Get Mat Instance Dynamic()
    Overlay Material=Get Overlay Mat Instance Dynamic()
    Skeletal Mesh=Get SkelMesh()
    "Local Noise Transform  Location"=Get Local Noise Transform Location() -> (Local Noise Transform Location:Vector)
    Local Noise Offset=Get Local Noise Offset()
   )
   > Set WPO Regional Function!(
    execute
    Material:Material Instance Dynamic Object Reference=Get Mat Instance Dynamic()
    Overlay Material:Material Instance Dynamic Object Reference=Get Overlay Mat Instance Dynamic()
    Skeletal Mesh:Skeletal Mesh Component Object Reference=Get SkelMesh()
    Velocity Transform:Name=foot_r
    Velocity Interp Speed:double=Get Velocity Interp Speed()
    Collision Transform:Name=calf_r
    Collision Transform Axis:EAxis Enum
    Invert Collision Transform Axis:bool=false
    Delta Time X:double=@n5.DeltaTimeX
    Previous Velocity Transform Location:"Vector (by ref)"=Get R1 Velocity Transform Location()
    Previous Velocity Transform Velocity:"Vector (by ref)"=Get R1 Velocity Transform Velocity()
    Region:int=1
   ) -> (then)
   > Set WPO Regional Function(
    Material=Get Mat Instance Dynamic()
    Overlay Material=Get Overlay Mat Instance Dynamic()
    Skeletal Mesh=Get SkelMesh()
    Velocity Transform=foot_l
    Velocity Interp Speed=Get Velocity Interp Speed()
    Collision Transform=calf_l
    Invert Collision Transform Axis=true
    Delta Time X=@n5.DeltaTimeX
    Previous Velocity Transform Location=Get R2 Velocity Transform Location()
    Previous Velocity Transform Velocity=Get R2 Velocity Transform Velocity()
    Region=2
   )
   > n7 = Branch(Condition=Get bIsSkyDiving())
   > Set WPO Pho Wind Intensity(WPO Pho Wind Intensity=0.8)
   > n8 = Set WPO Pho Turbulence(WPO Pho Turbulence:double=1.5) -> (Output_Get:double)
   SetScalarParameterValue(execute=[@n10.then, @n8.then], self=Get Mat Instance Dynamic(), ParameterName=ClothSim_WindStrength, Value=Get WPO Pho Wind Intensity())
   n9 = Get WPO Pho Turbulence() -> (WPO Pho Turbulence:double)
   > SetScalarParameterValue(self=Get Mat Instance Dynamic(), ParameterName=ClothSim_R1_TurbulenceStrength, Value=@n9)
   > SetScalarParameterValue(self=Get Mat Instance Dynamic(), ParameterName=ClothSim_R2_TurbulenceStrength, Value=@n9)
   Set WPO Pho Wind Intensity(execute=@n7.else, WPO Pho Wind Intensity=0)
   > n10 = Set WPO Pho Turbulence(WPO Pho Turbulence=0.2)
   n11 = Gravity Override State Function DEPRECATED(
    execute=@n6.then_1
    CharacterPartAnimInstance=Self-Reference()
    GameTimeInSeconds=GetGameTimeInSeconds()
    GravityOverrideParams=Get Gravity Override Params()
   )
   > Set GravityOverride(GravityOverride=@n11)
   n12 = Get bIsCrouching()
   > Set IsNotCrouching(IsNotCrouching:double="To Float (Boolean)"(InBool=NOT Boolean(A=@n12))) -> (Output_Get:double)
   n13 = Get bIsCrouchSprinting()
   n14 = Get bIsMoving2D()
   n15 = Get bIsSprinting()
   > Set Is Crouch Sprinting(Is Crouch Sprinting:double="To Float (Boolean)"(InBool=OR Boolean!(A=@n13, B=@n14, C:bool=@n15) -> (ReturnValue))) -> (Output_Get:double)
   n16 = AND Boolean(A=Get bIsMoving2D(), B=Get bIsCrouching())
   > Set isCrouchingMovingSprinting(isCrouchingMovingSprinting:double="To Float (Boolean)"(InBool=OR Boolean!(A=@n13, B=@n16) -> (ReturnValue))) -> (Output_Get:double)
   n17 = Get bIsCrouching()
   n18 = Get bIsSkyDiving()
   n19 = Get bIsDBNO()
   > Set IsCrouchOrSkydivingOrDBNO(IsCrouchOrSkydivingOrDBNO:double="To Float (Boolean)"(InBool=OR Boolean!(A=@n17, B=@n18, C:bool=@n19) -> (ReturnValue))) -> (Output_Get:double)
   > Set IsCrouchingDBNO(IsCrouchingDBNO:bool=OR Boolean!(A=Get bIsCrouching(), B=Get bIsDBNO(), C=Get bIsSliding(), D:bool) -> (ReturnValue)) -> (Output_Get:bool)
   n20 = Get bIsPlayingEmote()
   > Set FENoEMote(FENoEMote:bool=AND Boolean(A=Get bIsInFrontEnd(), B=NOT Boolean(A=@n20))) -> (Output_Get:bool)
   n21 = Get FENoEMote() -> (FENoEMote:bool)
   > Set RgidbodyNOTFE(RgidbodyNOTFE:bool=AND Boolean(A=Get bRigidBodyEnabled(), B=NOT Boolean(A=@n21))) -> (Output_Get:bool)
   n22 = Get bIsInFrontEnd()
   n23 = Get bIsPlayingEmote()
   Set isEmoting(execute=@n6.then_2, isEmoting:double="To Float (Boolean)"(InBool=AND Boolean(A=@n22, B=@n23))) -> (Output_Get:double)
   n24 = Event OnEndSkydiving()
   Event OnCharacterCustomizationLoaded()
   > n25 = Set SkelMesh(SkelMesh=GetPartSkeletalMeshComponent())
   > n26 = Get Dynamic Material Instance(SkeletalMeshComponent=@n25)
   > Set Mat Instance Dynamic(Mat Instance Dynamic=@n26)
   > n27 = GetOrCreateDynamicOverlayMaterial(MeshComponent=@n25, MaterialName=None)
   > Set Overlay Mat Instance Dynamic(Overlay Mat Instance Dynamic=@n27)
  ]}
]
variables [
 BeltAAlpha {type=number}
 Body_Mesh {type=SkeletalMeshComponent}
 FENoEMote {type=boolean}
 Gravity Override Params {type=GravityOverrideParamsStruct}
 GravityOverride {type=Vector}
 Is Crouch Sprinting {type=number, default=1}
 Is Running {type=boolean}
 IsCrouchingDBNO {type=boolean}
 isCrouchingMovingSprinting {type=number}
 IsCrouchOrSkydivingOrDBNO {type=number}
 isEmoting {type=number}
 IsNotCrouching {type=number, default=1}
 isSkydiving {type=number}
 Local Noise Offset {type=Vector}
 Local Noise Transform Location {type=Vector}
 Mat Instance Dynamic {type=MaterialInstanceDynamic}
 Negate Is Running {type=number}
 Overlay Mat Instance Dynamic {type=MaterialInstanceDynamic}
 PlayerPawn {type=FortPlayerPawnAthena}
 R1 Velocity Transform Location {type=Vector}
 R1 Velocity Transform Velocity {type=Vector}
 R2 Velocity Transform Location {type=Vector}
 R2 Velocity Transform Velocity {type=Vector}
 RgidbodyNOTFE {type=boolean}
 SkelMesh {type=SkeletalMeshComponent}
 Velocity Interp Speed {type=number}
 WPO Pho Turbulence {type=number}
 WPO Pho Wind Intensity {type=number}
]
