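# Physics asset for the bag on the female medium Lilac body cosmetic, giving the bag and the hip area capsule collision.
# Pelvis and both thighs form the body volumes, with the thighs and bag master bone locked in all angular axes to the pelvis.
# The three-capsule dyn_bag_01 body hangs from the master through the only constraint without locked angular motion, leaving it as the moving part.
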PhysicsAsset /BRCosmetics/Characters/Player/Female/Medium/Bodies/F_MED_Lilac/Meshes/F_MED_Lilac_Bag_Physics
bodies {
 dyn_bag_01 [
  {
   shapeType=Capsule
   length=20
   radius=2.25
   shapeName=dyn_bag_01_capsule
   center=<22.5278 -0.9123 0.8234>
   rotation=<rot 7.6109 -4.0656 -17.3205>
  }
  {
   shapeType=Capsule
   length=5
   radius=3
   shapeName=dyn_bag_01_capsule
   center=<15.2827 0.2922 -14.3238>
   rotation=<rot -55.426 -27.8586 28.685>
  }
  {
   shapeType=Capsule
   length=5
   radius=3
   shapeName=dyn_bag_01_capsule
   center=<8.4643 -3.2148 15.059>
   rotation=<rot -73.5476 115.8458 -127.1433>
  }
 ]
 dyn_bag_Master [{shapeType=Capsule, length=20, radius=4, shapeName=dyn_bag_Master_capsule}]
 pelvis [{
   shapeType=Capsule
   length=7.4287
   radius=12.7066
   shapeName=pelvis_capsule
   center=<-0.0242 -2.0065 0.117>
   rotation=<rot 3.3365 -90.6898 -90>
  }]
 thigh_l [{shapeType=Capsule, length=40, radius=10, center=<-22.5408 -1.0748 0.3162>, rotation=<rot 1.6194 89.1694 -78.9429>}]
 thigh_r [{shapeType=Capsule, length=40, radius=10.6165, center=<21.1408 1.3649 -1.4807>, rotation=<rot 0 94.4549 -80.4749>}]
}
constraints [
 {
  bone1Name=thigh_l
  bone2Name=pelvis
  swing1LimitDegrees=45
  swing1Motion=Locked
  swing2LimitDegrees=45
  swing2Motion=Locked
  twistLimitDegrees=45
  twistMotion=Locked
 }
 {
  bone1Name=thigh_r
  bone2Name=pelvis
  swing1LimitDegrees=45
  swing1Motion=Locked
  swing2LimitDegrees=45
  swing2Motion=Locked
  twistLimitDegrees=45
  twistMotion=Locked
 }
 {
  bone1Name=dyn_bag_Master
  bone2Name=pelvis
  swing1LimitDegrees=45
  swing1Motion=Locked
  swing2LimitDegrees=45
  swing2Motion=Locked
  twistLimitDegrees=45
  twistMotion=Locked
 }
 {bone1Name=dyn_bag_01, bone2Name=dyn_bag_Master, swing1LimitDegrees=15, swing2LimitDegrees=1, twistLimitDegrees=1}
]
