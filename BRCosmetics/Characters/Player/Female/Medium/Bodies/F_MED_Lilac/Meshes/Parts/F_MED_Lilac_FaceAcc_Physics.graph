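# Physics asset for the FaceAcc mesh of the Lilac medium female player body, defining capsule bodies and joints for dynamic hair.
# Head, neck, upper spine, clavicle and upper-arm capsules form an angle-locked body frame, with hair master bones locked to the head.
# Five angle-limited chains hang from the masters: center, left, right and two F-series chains.
# Cross-constraints with free swing and twist link the side chains to the center at segments 02 and 04.
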PhysicsAsset /BRCosmetics/Characters/Player/Female/Medium/Bodies/F_MED_Lilac/Meshes/Parts/F_MED_Lilac_FaceAcc_Physics
bodies {
 clavicle_l [{shapeType=Capsule, length=6.25, radius=7, shapeName=clavicle_l_capsule, center=<6 -2 -1.5>, rotation=<rot 0 -90 -90>}]
 clavicle_r [{shapeType=Capsule, length=6.25, radius=6, shapeName=clavicle_l_capsule, center=<-6 2 1.5>, rotation=<rot 0 -90 -90>}]
 dyn_hair_01 [{shapeType=Capsule, length=6, radius=1.5, shapeName=dyn_hair_01_capsule, center=<6 0 0>}]
 dyn_hair_01_l [{shapeType=Capsule, length=6, radius=2, shapeName=dyn_hair_01_l_capsule, center=<6 0 0>, rotation=<rot -90 0 0>}]
 dyn_hair_01_r [{shapeType=Capsule, length=5, radius=2, shapeName=dyn_hair_01_l_capsule, center=<6 0 0>, rotation=<rot -90 0 160.0003>}]
 dyn_hair_02 [{shapeType=Capsule, length=6, radius=1.5, shapeName=dyn_hair_02_capsule, center=<3 0.5 0>}]
 dyn_hair_02_l [{shapeType=Capsule, length=2.5, radius=1.5, shapeName=dyn_hair_02_l_capsule, center=<3 1 0>, rotation=<rot -90 0 0>}]
 dyn_hair_02_r [{shapeType=Capsule, length=2.5, radius=1.5, shapeName=dyn_hair_02_r_capsule, center=<3 0.5 0>, rotation=<rot -90 180 180>}]
 dyn_hair_03 [{shapeType=Capsule, length=6, radius=1.5, shapeName=dyn_hair_03_capsule, center=<1 0 0>}]
 dyn_hair_03_l [{shapeType=Capsule, length=3.5, radius=1.75, shapeName=dyn_hair_03_l_capsule, center=<3.5 0 0>, rotation=<rot -90 0 0>}]
 dyn_hair_03_r [{shapeType=Capsule, length=3.5, radius=1.5, shapeName=dyn_hair_03_r_capsule, center=<1.5 0.38 -2>, rotation=<rot -90 0 0>}]
 dyn_hair_04 [{shapeType=Capsule, length=6, radius=1.5, shapeName=dyn_hair_04_capsule, center=<1 0 0>}]
 dyn_hair_04_l [{shapeType=Capsule, length=5, radius=1.5, shapeName=dyn_hair_04_l_capsule, center=<5 0 0>, rotation=<rot -90 0 0>}]
 dyn_hair_04_r [{shapeType=Capsule, length=5, radius=1.5, shapeName=dyn_hair_04_r_capsule, center=<1 0.15 0>, rotation=<rot -90 180 180>}]
 dyn_hair_05 [{shapeType=Capsule, length=6, radius=1.5, shapeName=dyn_hair_05_capsule, center=<4.1171 0 0.5662>}]
 dyn_hair_F00_l [{shapeType=Capsule, length=3, radius=1.25, shapeName=dyn_hair_F00_l_capsule, center=<1 0 0>, rotation=<rot -90 0 0>}]
 dyn_hair_F00_r [{shapeType=Capsule, length=3, radius=1, shapeName=dyn_hair_F00_r_capsule, center=<1 0 0>, rotation=<rot -90 0 0>}]
 dyn_hair_F01_l [{shapeType=Capsule, length=3, radius=1.25, shapeName=dyn_hair_F01_l_capsule, center=<1 0.4132 0>, rotation=<rot -90 -156.653 166.6528>}]
 dyn_hair_F01_r [{shapeType=Capsule, length=3, radius=1.5, shapeName=dyn_hair_F01_r_capsule, center=<1 0.6 0>, rotation=<rot -90 0 0>}]
 dyn_hair_F02_l [{shapeType=Capsule, length=2, radius=1.5, shapeName=dyn_hair_F02_l_capsule, center=<0.6279 0.7376 0>, rotation=<rot -90 180 180>}]
 dyn_hair_F02_r [{shapeType=Capsule, length=3, radius=1.5, shapeName=dyn_hair_F02_r_capsule, center=<-1.5781 1.1369 0>, rotation=<rot -90 0 0>}]
 dyn_hair_F03_l [{
   shapeType=Capsule
   length=9
   radius=1.25
   shapeName=dyn_hair_F03_l_capsule
   center=<3.0966 -0.6137 -0.1027>
   rotation=<rot -80 -30.0008 0.0001>
  }]
 dyn_hair_F03_r [{shapeType=Capsule, length=8, radius=1.5, shapeName=dyn_hair_F03_r_capsule, center=<0.2136 0.96 0>, rotation=<rot -80 149.9995 -179.9998>}]
 hair_01_MASTER [{shapeType=Capsule, length=0.5, radius=0.5, shapeName=Earing_MASTER_r_capsule}]
 hair_01_MASTER_l [{shapeType=Capsule, length=0.5, radius=0.5, shapeName=Earing_MASTER_r_capsule}]
 hair_01_MASTER_r [{shapeType=Capsule, length=0.5, radius=0.5, shapeName=Earing_MASTER_r_capsule}]
 hair_F00_MASTER_l [{shapeType=Capsule, length=0.505, radius=0.505, shapeName=hair_F00_MASTER_l_capsule}]
 hair_F00_MASTER_r [{shapeType=Capsule, length=0.505, radius=0.505, shapeName=hair_F00_MASTER_r_capsule}]
 head [{shapeType=Capsule, length=9, radius=5.5, rotation=<rot -90 0 0>}]
 neck_01 [{shapeType=Capsule, length=25, radius=4, shapeName=neck_01_capsule, center=<2.5 0 0>}]
 neck_02 [{shapeType=Capsule, length=5, radius=4, shapeName=neck_02_capsule, rotation=<rot 0 -90 -90>}]
 spine_05 [{shapeType=Capsule, length=7.5, radius=10, shapeName=spine_05_capsule, center=<-2 0 0>}]
 upperarm_l [{shapeType=Capsule, length=10, radius=4, shapeName=upperarm_l_capsule, center=<9 0 0>, rotation=<rot 90 0 0>}]
 upperarm_r [{shapeType=Capsule, length=10, radius=4, shapeName=upperarm_r_capsule, center=<-9 0 0>, rotation=<rot 90 0 0>}]
}
constraints [
 {
  bone1Name=head
  bone2Name=neck_02
  swing1LimitDegrees=45
  swing1Motion=Locked
  swing2LimitDegrees=45
  swing2Motion=Locked
  twistLimitDegrees=45
  twistMotion=Locked
 }
 {
  bone1Name=neck_01
  bone2Name=spine_05
  swing1LimitDegrees=45
  swing1Motion=Locked
  swing2LimitDegrees=45
  swing2Motion=Locked
  twistLimitDegrees=45
  twistMotion=Locked
 }
 {
  bone1Name=clavicle_r
  bone2Name=spine_05
  swing1LimitDegrees=45
  swing1Motion=Locked
  swing2LimitDegrees=45
  swing2Motion=Locked
  twistLimitDegrees=45
  twistMotion=Locked
 }
 {
  bone1Name=clavicle_l
  bone2Name=spine_05
  swing1LimitDegrees=45
  swing1Motion=Locked
  swing2LimitDegrees=45
  swing2Motion=Locked
  twistLimitDegrees=45
  twistMotion=Locked
 }
 {
  bone1Name=hair_01_MASTER
  bone2Name=head
  swing1LimitDegrees=45
  swing1Motion=Locked
  swing2LimitDegrees=45
  swing2Motion=Locked
  twistLimitDegrees=45
  twistMotion=Locked
 }
 {
  bone1Name=hair_01_MASTER_l
  bone2Name=head
  swing1LimitDegrees=45
  swing1Motion=Locked
  swing2LimitDegrees=45
  swing2Motion=Locked
  twistLimitDegrees=45
  twistMotion=Locked
 }
 {
  bone1Name=hair_01_MASTER_r
  bone2Name=head
  swing1LimitDegrees=45
  swing1Motion=Locked
  swing2LimitDegrees=45
  swing2Motion=Locked
  twistLimitDegrees=45
  twistMotion=Locked
 }
 {
  bone1Name=hair_F00_MASTER_l
  bone2Name=head
  swing1LimitDegrees=45
  swing1Motion=Locked
  swing2LimitDegrees=45
  swing2Motion=Locked
  twistLimitDegrees=45
  twistMotion=Locked
 }
 {
  bone1Name=hair_F00_MASTER_r
  bone2Name=head
  swing1LimitDegrees=45
  swing1Motion=Locked
  swing2LimitDegrees=45
  swing2Motion=Locked
  twistLimitDegrees=45
  twistMotion=Locked
 }
 {
  bone1Name=upperarm_l
  bone2Name=clavicle_l
  swing1LimitDegrees=45
  swing1Motion=Locked
  swing2LimitDegrees=45
  swing2Motion=Locked
  twistLimitDegrees=45
  twistMotion=Locked
 }
 {
  bone1Name=upperarm_r
  bone2Name=clavicle_r
  swing1LimitDegrees=45
  swing1Motion=Locked
  swing2LimitDegrees=45
  swing2Motion=Locked
  twistLimitDegrees=45
  twistMotion=Locked
 }
 {bone1Name=dyn_hair_F00_l, bone2Name=hair_F00_MASTER_l, swing1LimitDegrees=25, swing2LimitDegrees=15, twistLimitDegrees=1}
 {bone1Name=dyn_hair_F01_l, bone2Name=dyn_hair_F00_l, swing1LimitDegrees=15, swing2LimitDegrees=15, twistLimitDegrees=5}
 {bone1Name=dyn_hair_F02_l, bone2Name=dyn_hair_F01_l, swing1LimitDegrees=15, swing2LimitDegrees=15, twistLimitDegrees=5}
 {bone1Name=dyn_hair_F03_l, bone2Name=dyn_hair_F02_l, swing1LimitDegrees=10, swing2LimitDegrees=10, twistLimitDegrees=5}
 {bone1Name=dyn_hair_01_r, bone2Name=hair_01_MASTER_r, swing1LimitDegrees=70, swing2LimitDegrees=5, twistLimitDegrees=0}
 {bone1Name=dyn_hair_02_r, bone2Name=dyn_hair_01_r, swing1LimitDegrees=40, swing2LimitDegrees=10, twistLimitDegrees=10}
 {bone1Name=dyn_hair_03_r, bone2Name=dyn_hair_02_r, swing1LimitDegrees=30, swing2LimitDegrees=15, twistLimitDegrees=5, twistMotion=Locked}
 {bone1Name=dyn_hair_04_r, bone2Name=dyn_hair_03_r, swing1LimitDegrees=20, swing2LimitDegrees=20, twistLimitDegrees=5, twistMotion=Locked}
 {bone1Name=dyn_hair_01, bone2Name=hair_01_MASTER, swing1LimitDegrees=20, swing2LimitDegrees=5, twistLimitDegrees=0}
 {bone1Name=dyn_hair_02, bone2Name=dyn_hair_01, swing1LimitDegrees=40, swing2LimitDegrees=10, twistLimitDegrees=20}
 {bone1Name=dyn_hair_03, bone2Name=dyn_hair_02, swing1LimitDegrees=30, swing2LimitDegrees=15, twistLimitDegrees=5, twistMotion=Locked}
 {bone1Name=dyn_hair_04, bone2Name=dyn_hair_03, swing1LimitDegrees=20, swing2LimitDegrees=20, twistLimitDegrees=5, twistMotion=Locked}
 {bone1Name=dyn_hair_05, bone2Name=dyn_hair_04, swing1LimitDegrees=20, swing2LimitDegrees=20, twistLimitDegrees=5, twistMotion=Locked}
 {bone1Name=dyn_hair_01_l, bone2Name=hair_01_MASTER_l, swing1LimitDegrees=70, swing2LimitDegrees=5, twistLimitDegrees=0, twistMotion=Locked}
 {bone1Name=dyn_hair_02_l, bone2Name=dyn_hair_01_l, swing1LimitDegrees=40, swing2LimitDegrees=10, twistLimitDegrees=1}
 {bone1Name=dyn_hair_03_l, bone2Name=dyn_hair_02_l, swing1LimitDegrees=30, swing2LimitDegrees=15, twistLimitDegrees=5, twistMotion=Locked}
 {bone1Name=dyn_hair_04_l, bone2Name=dyn_hair_03_l, swing1LimitDegrees=20, swing2LimitDegrees=20, twistLimitDegrees=5, twistMotion=Locked}
 {
  bone1Name=dyn_hair_02_l
  bone2Name=dyn_hair_02
  swing1LimitDegrees=45
  swing1Motion=Free
  swing2LimitDegrees=45
  swing2Motion=Free
  twistLimitDegrees=45
  twistMotion=Free
 }
 {
  bone1Name=dyn_hair_02_r
  bone2Name=dyn_hair_02
  swing1LimitDegrees=45
  swing1Motion=Free
  swing2LimitDegrees=45
  swing2Motion=Free
  twistLimitDegrees=45
  twistMotion=Free
 }
 {
  bone1Name=dyn_hair_04_l
  bone2Name=dyn_hair_04
  swing1LimitDegrees=45
  swing1Motion=Free
  swing2LimitDegrees=45
  swing2Motion=Free
  twistLimitDegrees=45
  twistMotion=Free
 }
 {
  bone1Name=dyn_hair_04_r
  bone2Name=dyn_hair_04
  swing1LimitDegrees=45
  swing1Motion=Free
  swing2LimitDegrees=45
  swing2Motion=Free
  twistLimitDegrees=45
  twistMotion=Free
 }
 {bone1Name=dyn_hair_F00_r, bone2Name=hair_F00_MASTER_r, swing1LimitDegrees=30, swing2LimitDegrees=10, twistLimitDegrees=5}
 {bone1Name=dyn_hair_F01_r, bone2Name=dyn_hair_F00_r, swing1LimitDegrees=15, swing2LimitDegrees=10, twistLimitDegrees=5}
 {bone1Name=dyn_hair_F02_r, bone2Name=dyn_hair_F01_r, swing1LimitDegrees=15, swing2LimitDegrees=10, twistLimitDegrees=5}
 {bone1Name=dyn_hair_F03_r, bone2Name=dyn_hair_F02_r, swing1LimitDegrees=10, swing2LimitDegrees=10, twistLimitDegrees=1}
 {
  bone1Name=neck_02
  bone2Name=neck_01
  swing1LimitDegrees=45
  swing1Motion=Locked
  swing2LimitDegrees=45
  swing2Motion=Locked
  twistLimitDegrees=45
  twistMotion=Locked
 }
]
