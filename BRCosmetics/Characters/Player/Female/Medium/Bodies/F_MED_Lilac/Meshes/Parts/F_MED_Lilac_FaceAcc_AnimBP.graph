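# Animation Blueprint for the F_MED_Lilac face accessory, adding secondary motion to hair and an earring.
# Copied parent and body poses run through AnimDynamics, then bone rotations driven by melee, targeting and head-to-spine angle.
# That cached pose feeds neck-relative RigidBody simulations, selected by front end, cape or long backbling, other backbling, or none.
# The event graph detects backbling and updates state-based gravity, skydive-dependent inertia scales and the hair-lift alpha.
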AnimBlueprint /BRCosmetics/Characters/Player/Female/Medium/Bodies/F_MED_Lilac/Meshes/Parts/F_MED_Lilac_FaceAcc_AnimBP
target_skeleton=&/F_MED_Lilac_FaceAcc_Skeleton
graphs [
 AnimGraph {nodes [
   n1 = Get Linear Scale WO() -> (Linear Scale WO:Vector)
   n2 = RigidBody!(
    ComponentPose=Local To Component(LocalPose=Use cached pose 'Skeleton Blend'())
    Alpha
    OverrideWorldGravity:Vector=Get GravityOverride()
    ExternalForce
    ComponentLinearAccScale:Vector=@n1
    ComponentLinearVelScale:Vector=@n1
   ) -> (Pose) {
    bClampLinearTranslationLimitToRefPose=true
    bOverrideWorldGravity=true
    overridePhysicsAsset=&/F_MED_Lilac_FaceAccFrontEnd_Physics
    simulationSpace=BaseBoneSpace
    alphaBoolBlend {blendOption=Linear}
    alphaScaleBiasClamp {bInterpResult=true, interpSpeedIncreasing=25}
    baseBoneRef {boneName=neck_01}
    componentAppliedLinearAccClamp=<3000 3000 3000>
    componentLinearAccScale=<0.5 0.2 0.2>
    componentLinearVelScale=<0.5 0.2 0.2>
    overrideWorldGravity {z=-780}
   }
   Output Pose(Result=Blend Poses by bool(BlendPose_0=Component To Local(ComponentPose=@n2), BlendPose_1=Use cached pose 'CapePose'(), bActiveValue=Get bIsInFrontEnd()) {blendType=Linear}) {layerGroup=None}
   n3 = Get bIsPlayingMeleeAnim()
   n4 = Anim Dynamics - Bone: hair_A01_MASTER(ComponentPose=Local To Component(LocalPose=Use cached pose 'StartPose'()), Alpha=1, ExternalForce="0,0,0") {
    angularDampingOverride=0.999
    angularSpringConstant=100
    bAngularSpring=true
    bGravityOverrideInSimSpace=true
    bOverrideAngularDamping=true
    bUseGravityOverride=true
    alphaBoolBlend {blendOption=Linear}
    boundBone {boneName=hair_A01_MASTER}
    componentAppliedLinearAccClamp=<3000 3000 3000>
    componentLinearAccScale=<0.3 0.3 0.3>
    componentLinearVelScale=<0.3 0.3 0.3>
    gravityOverride {z=-980}
    physicsBodyDefinitions [{
      sphereCollisionRadius=0
      boundBone {boneName=hair_A01_MASTER}
      boxExtents=<40 40 10>
      constraintSetup {angularLimitsMax=<1 1 0>, angularLimitsMin=<-1 -1 -20>, angularTarget=<1 0 0>}
      localJointOffset=<2 0.2 0.2>
     }]
   }
   n5 = Anim Dynamics - Bone: hair_B01_MASTER(ComponentPose=@n4, Alpha=1, ExternalForce="0,0,0") {
    angularDampingOverride=0.999
    angularSpringConstant=100
    bAngularSpring=true
    bGravityOverrideInSimSpace=true
    bOverrideAngularDamping=true
    bUseGravityOverride=true
    alphaBoolBlend {blendOption=Linear}
    boundBone {boneName=hair_B01_MASTER}
    componentAppliedLinearAccClamp=<3000 3000 3000>
    componentLinearAccScale=<0.3 0.3 0.3>
    componentLinearVelScale=<0.3 0.3 0.3>
    gravityOverride {z=-980}
    physicsBodyDefinitions [{
      sphereCollisionRadius=0
      boundBone {boneName=hair_B01_MASTER}
      boxExtents=<30 20 30>
      constraintSetup {angularLimitsMax=<1 1 0>, angularLimitsMin=<-1 -1 -50>, angularTarget=<0 0 0>}
      localJointOffset=<2 0.2 0.2>
     }]
   }
   n6 = Anim Dynamics - Bone: Earing_MASTER_r(ComponentPose=@n5, Alpha=1, ExternalForce="0,0,0") {
    angularDampingOverride=0.999
    angularSpringConstant=100
    bAngularSpring=true
    bGravityOverrideInSimSpace=true
    bOverrideAngularDamping=true
    bUseGravityOverride=true
    alphaBoolBlend {blendOption=Linear}
    boundBone {boneName=Earing_MASTER_r}
    componentAppliedLinearAccClamp=<3000 3000 3000>
    componentLinearAccScale=<0.3 0.3 0.3>
    componentLinearVelScale=<0.3 0.3 0.3>
    gravityOverride {z=-980}
    physicsBodyDefinitions [{
      sphereCollisionRadius=0
      boundBone {boneName=Earing_MASTER_r}
      boxExtents=<10 12 12>
      constraintSetup {angularLimitsMax=<30 10 30>, angularLimitsMin=<-1 0 -30>, angularTarget=<0 0 0>}
      localJointOffset=<2 0.2 0>
     }]
   }
   n7 = Anim Dynamics - Bone: hair_strands_01_MASTER(ComponentPose=@n6, Alpha=1, ExternalForce="0,0,0") {
    angularSpringConstant=50
    bAngularSpring=true
    bUseGravityOverride=true
    alphaBoolBlend {blendOption=Linear}
    boundBone {boneName=hair_strands_01_MASTER}
    gravityOverride {z=-98}
    physicsBodyDefinitions [{
      boundBone {boneName=hair_strands_01_MASTER}
      boxExtents=<15 10 10>
      constraintSetup {angularLimitsMax=<0 20 20>, angularLimitsMin=<0 -20 -20>, angularTarget=<0 0 0>}
      localJointOffset=<1 0 0>
     }]
   }
   n8 = Anim Dynamics - Bone: hair_strands_02_MASTER(ComponentPose=@n7, Alpha=1, ExternalForce="0,0,0") {
    angularDampingOverride=0.999
    bGravityOverrideInSimSpace=true
    bOverrideAngularDamping=true
    bUseGravityOverride=true
    alphaBoolBlend {blendOption=Linear}
    boundBone {boneName=hair_strands_02_MASTER}
    componentAppliedLinearAccClamp=<3000 3000 3000>
    componentLinearAccScale=<0.3 0.3 0.3>
    componentLinearVelScale=<0.3 0.3 0.3>
    gravityOverride {z=-98}
    physicsBodyDefinitions [{
      sphereCollisionRadius=0
      boundBone {boneName=hair_strands_02_MASTER}
      boxExtents=<10 12 12>
      constraintSetup {angularLimitsMax=<1 10 20>, angularLimitsMin=<-1 -10 -20>, angularTarget=<0 0 0>}
      localJointOffset=<2 0.2 0.2>
     }]
   }
   n9 = Anim Dynamics - Bone: hair_strands_03_MASTER(ComponentPose=@n8, Alpha=1, ExternalForce="0,0,0") {
    angularDampingOverride=0.999
    bGravityOverrideInSimSpace=true
    bOverrideAngularDamping=true
    bUseGravityOverride=true
    alphaBoolBlend {blendOption=Linear}
    boundBone {boneName=hair_strands_03_MASTER}
    componentAppliedLinearAccClamp=<3000 3000 3000>
    componentLinearAccScale=<0.3 0.3 0.3>
    componentLinearVelScale=<0.3 0.3 0.3>
    gravityOverride {z=-700}
    physicsBodyDefinitions [{
      sphereCollisionRadius=0
      boundBone {boneName=hair_strands_03_MASTER}
      boxExtents=<10 12 12>
      constraintSetup {angularLimitsMax=<1 15 0>, angularLimitsMin=<-1 -10 0>, angularTarget=<0 0 0>}
      localJointOffset=<2 0.2 0.2>
     }]
   }
   n10 = Anim Dynamics - Bone: hair_strands_04_MASTER(ComponentPose=@n9, Alpha=1, ExternalForce="0,0,0") {
    angularDampingOverride=0.999
    bGravityOverrideInSimSpace=true
    bOverrideAngularDamping=true
    bUseGravityOverride=true
    alphaBoolBlend {blendOption=Linear}
    boundBone {boneName=hair_strands_04_MASTER}
    componentAppliedLinearAccClamp=<3000 3000 3000>
    componentLinearAccScale=<0.3 0.3 0.3>
    componentLinearVelScale=<0.3 0.3 0.3>
    gravityOverride {z=-700}
    physicsBodyDefinitions [{
      sphereCollisionRadius=0
      boundBone {boneName=hair_strands_04_MASTER}
      boxExtents=<10 12 12>
      constraintSetup {angularLimitsMax=<1 0 10>, angularLimitsMin=<-1 0 -20>, angularTarget=<0 0 0>}
      localJointOffset=<2 0.2 0.2>
     }]
   }
   n11 = Anim Dynamics - Bone: hair_strands_05_MASTER(ComponentPose=@n10, Alpha=1, ExternalForce="0,0,0") {
    angularDampingOverride=0.999
    bGravityOverrideInSimSpace=true
    bOverrideAngularDamping=true
    bUseGravityOverride=true
    alphaBoolBlend {blendOption=Linear}
    boundBone {boneName=hair_strands_05_MASTER}
    componentAppliedLinearAccClamp=<3000 3000 3000>
    componentLinearAccScale=<0.3 0.3 0.3>
    componentLinearVelScale=<0.3 0.3 0.3>
    gravityOverride {z=-700}
    physicsBodyDefinitions [{
      sphereCollisionRadius=0
      boundBone {boneName=hair_strands_05_MASTER}
      boxExtents=<10 12 12>
      constraintSetup {angularLimitsMax=<1 10 20>, angularLimitsMin=<-1 0 -20>, angularTarget=<0 0 0>}
      localJointOffset=<2 0.2 0.2>
     }]
   }
   n12 = "Transform (Modify) Bone - Bone: hair_01_MASTER_l"!(ComponentPose=@n11, bAlphaBoolEnabled:bool=@n3, Translation="0,0,0") -> (Pose) {
    alphaInputType=Bool
    rotationMode=BMM_Additive
    rotationSpace=BCS_BoneSpace
    translationSpace=BCS_BoneSpace
    alphaBoolBlend {blendOption=Linear}
    alphaScaleBiasClamp {bInterpResult=true, interpSpeedDecreasing=5, interpSpeedIncreasing=7}
    boneToModify {boneName=hair_01_MASTER_l}
    rotation {yaw=-30}
   }
   n13 = "Transform (Modify) Bone - Bone: hair_01_MASTER_r"!(ComponentPose=@n12, bAlphaBoolEnabled:bool=@n3, Translation="0,0,0") -> (Pose) {
    alphaInputType=Bool
    rotationMode=BMM_Additive
    rotationSpace=BCS_BoneSpace
    translationSpace=BCS_BoneSpace
    alphaBoolBlend {blendOption=Linear}
    alphaScaleBiasClamp {bInterpResult=true, interpSpeedDecreasing=5, interpSpeedIncreasing=7}
    boneToModify {boneName=hair_01_MASTER_r}
    rotation {yaw=-30}
   }
   n14 = "Transform (Modify) Bone - Bone: hair_01_MASTER"!(ComponentPose=@n13, bAlphaBoolEnabled:bool=@n3, Translation="0,0,0") -> (Pose) {
    alphaInputType=Bool
    rotationMode=BMM_Additive
    rotationSpace=BCS_BoneSpace
    translationSpace=BCS_BoneSpace
    alphaBoolBlend {blendOption=Linear}
    alphaScaleBiasClamp {bInterpResult=true, interpSpeedDecreasing=5, interpSpeedIncreasing=7}
    boneToModify {boneName=hair_01_MASTER}
    rotation {yaw=-30}
   }
   n15 = Get bIsTargeting()
   n16 = "Transform (Modify) Bone - Bone: hair_01_MASTER_l"(ComponentPose=@n14, bAlphaBoolEnabled=@n15, Translation="0,0,0") {
    alphaInputType=Bool
    rotationMode=BMM_Additive
    rotationSpace=BCS_BoneSpace
    translationSpace=BCS_BoneSpace
    alphaBoolBlend {blendOption=Linear}
    alphaScaleBiasClamp {bInterpResult=true, interpSpeedDecreasing=5, interpSpeedIncreasing=7}
    boneToModify {boneName=hair_01_MASTER_l}
   }
   n17 = "Transform (Modify) Bone - Bone: hair_01_MASTER_r"(ComponentPose=@n16, bAlphaBoolEnabled=@n15, Translation="0,0,0") {
    alphaInputType=Bool
    rotationMode=BMM_Additive
    rotationSpace=BCS_BoneSpace
    translationMode=BMM_Additive
    translationSpace=BCS_BoneSpace
    alphaBoolBlend {blendOption=Linear}
    alphaScaleBiasClamp {interpSpeedDecreasing=5, interpSpeedIncreasing=7}
    boneToModify {boneName=hair_01_MASTER_r}
    rotation {pitch=20, yaw=-20}
   }
   n18 = Get Hair Up() -> (Hair Up:double)
   n19 = "Transform (Modify) Bone - Bone: hair_01_MASTER_l"!(ComponentPose=@n17, Alpha:float=@n18, Translation="0,0,0") -> (Pose) {
    rotationMode=BMM_Additive
    rotationSpace=BCS_BoneSpace
    translationSpace=BCS_BoneSpace
    alphaBoolBlend {blendOption=Linear}
    alphaScaleBiasClamp {bInterpResult=true, interpSpeedDecreasing=5, interpSpeedIncreasing=7}
    boneToModify {boneName=hair_01_MASTER_l}
    rotation {yaw=-90}
   }
   n20 = "Transform (Modify) Bone - Bone: hair_01_MASTER_r"!(ComponentPose=@n19, Alpha:float=@n18, Translation="0,0,0") -> (Pose) {
    rotationMode=BMM_Additive
    rotationSpace=BCS_BoneSpace
    translationSpace=BCS_BoneSpace
    alphaBoolBlend {blendOption=Linear}
    alphaScaleBiasClamp {bInterpResult=true, interpSpeedDecreasing=5, interpSpeedIncreasing=7}
    boneToModify {boneName=hair_01_MASTER_r}
    rotation {yaw=-90}
   }
   n21 = "Transform (Modify) Bone - Bone: hair_01_MASTER"!(ComponentPose=@n20, Alpha:float=@n18, Translation="0,0,0") -> (Pose) {
    rotationMode=BMM_Additive
    rotationSpace=BCS_BoneSpace
    translationSpace=BCS_BoneSpace
    alphaBoolBlend {blendOption=Linear}
    alphaScaleBiasClamp {bInterpResult=true, interpSpeedDecreasing=5, interpSpeedIncreasing=7}
    boneToModify {boneName=hair_01_MASTER}
    rotation {yaw=-90}
   }
   Save cached pose 'Skeleton Blend'(Pose=Component To Local(ComponentPose=@n21)) {cacheName=Skeleton Blend}
   n22 = Get Linear Scale WO()
   n23 = RigidBody(
    ComponentPose=Local To Component(LocalPose=Use cached pose 'Skeleton Blend'())
    OverrideWorldGravity=Get GravityOverride()
    ComponentLinearAccScale=@n22
    ComponentLinearVelScale=@n22
   ) {
    bClampLinearTranslationLimitToRefPose=true
    bOverrideWorldGravity=true
    overridePhysicsAsset=&/F_MED_Lilac_FaceAcc_Cape_Physics
    simulationSpace=BaseBoneSpace
    alphaBoolBlend {blendOption=Linear}
    alphaScaleBiasClamp {bInterpResult=true, interpSpeedIncreasing=25}
    baseBoneRef {boneName=neck_01}
    componentAppliedLinearAccClamp=<3000 3000 3000>
    componentLinearAccScale=<0.5 0.2 0.2>
    componentLinearVelScale=<0.5 0.2 0.2>
    overrideWorldGravity {z=-780}
   }
   Save cached pose 'CapePose'(Pose=Blend Poses by bool(
     BlendPose_0=Component To Local(ComponentPose=@n23)
     BlendPose_1=Use cached pose 'BackpackPose'()
     bActiveValue=Get Is Wearing Cape() -> (Is Wearing Cape:bool)
    ) {blendType=Linear}) {cacheName=CapePose}
   n24 = RigidBody(
    ComponentPose=Local To Component(LocalPose=Use cached pose 'Skeleton Blend'())
    OverrideWorldGravity="0,0,-780"
    ComponentLinearAccScale="0.1,0.2,0.2"
    ComponentLinearVelScale="0.1,0.2,0.2"
   ) {
    bClampLinearTranslationLimitToRefPose=true
    bOverrideWorldGravity=true
    overridePhysicsAsset=&/F_MED_Lilac_FaceAccBackpack_Physics
    simulationSpace=BaseBoneSpace
    alphaBoolBlend {blendOption=Linear}
    alphaScaleBiasClamp {bInterpResult=true, interpSpeedIncreasing=25}
    baseBoneRef {boneName=neck_01}
    componentAppliedLinearAccClamp=<3000 3000 3000>
    componentLinearAccScale=<0.32 0.2 0.2>
    componentLinearVelScale=<0.3 0.2 0.2>
    overrideWorldGravity {z=-780}
   }
   n25 = "Transform (Modify) Bone - Bone: hair_01_MASTER_r"!(
    ComponentPose=Local To Component(LocalPose=Use cached pose 'Skeleton Blend'())
    bAlphaBoolEnabled:bool=Get bIsSkydiveFloating()
    Translation="0,0,0"
   ) -> (Pose) {
    alphaInputType=Bool
    rotationMode=BMM_Additive
    rotationSpace=BCS_BoneSpace
    translationMode=BMM_Additive
    translationSpace=BCS_BoneSpace
    alphaBoolBlend {blendOption=Linear}
    alphaScaleBiasClamp {interpSpeedDecreasing=5, interpSpeedIncreasing=7}
    boneToModify {boneName=hair_01_MASTER_r}
    rotation {yaw=10}
   }
   n26 = Get Linear Scale WO()
   n27 = RigidBody(ComponentPose=@n25, OverrideWorldGravity=Get GravityOverride(), ComponentLinearAccScale=@n26, ComponentLinearVelScale=@n26) {
    bClampLinearTranslationLimitToRefPose=true
    bOverrideWorldGravity=true
    overridePhysicsAsset=&/F_MED_Lilac_FaceAcc_Physics
    simulationSpace=BaseBoneSpace
    alphaBoolBlend {blendOption=Linear}
    alphaScaleBiasClamp {bInterpResult=true, interpSpeedIncreasing=25}
    baseBoneRef {boneName=neck_01}
    componentAppliedLinearAccClamp=<3000 3000 3000>
    componentLinearAccScale=<0.5 0.2 0.2>
    componentLinearVelScale=<0.5 0.2 0.2>
    overrideWorldGravity {z=-780}
   }
   Save cached pose 'BackpackPose'(Pose=Blend Poses by bool(
     BlendPose_0=Component To Local(ComponentPose=@n24)
     BlendPose_1=Component To Local(ComponentPose=@n27)
     bActiveValue=Get isWearingBackbling()
    ) {blendType=Linear}) {cacheName=BackpackPose}
   n28 = Layered blend per bone(
    BasePose=Copy Pose From Mesh() {bCopyCurves=true, bUseAttachedParent=true}
    BlendPoses_0=Copy Pose From Mesh(SourceMeshComponent=Get Body_Mesh())
   ) {layerSetup [{branchFilters [{boneName=clavicle_l}, {boneName=clavicle_r}, {boneName=upperarm_r}, {boneName=upperarm_l}]}]}
   Save cached pose 'StartPose'(Pose=@n28) {cacheName=StartPose}
   Get Linear Scale W() -> (Linear Scale W:Vector)
  ]}
 EventGraph {nodes [
   Event OnCharacterCustomizationLoaded()
   > n1 = Cast To PlayerPawn_Athena(Object=TryGetPawnOwner())
   > Set Body_Mesh(Body_Mesh=GetSkeletalMeshForPartType(self=@n1))
   n2 = CheckForBackbling()
   > Set isWearingBackbling(isWearingBackbling=AND Boolean!(A=@n2, B=NOT Boolean(A=@n2.bIsBackblingACape), C:bool=NOT Boolean(A=@n2.bIsBackblingLong)) -> (ReturnValue))
   > Set Is Wearing Cape(Is Wearing Cape=OR Boolean(A=@n2.bIsBackblingACape, B=@n2.bIsBackblingLong))
   Event BlueprintInitializeAnimation()
   > Set Skel Mesh(Skel Mesh=GetOwnerSkeletalMeshComponent())
   Event BlueprintUpdateAnimation()
   > n3 = Sequence!(execute) -> (then_0, then_1, then_2:Exec, then_3:Exec, then_4:Exec, then_5:Exec, then_6:Exec)
   n4 = OR Boolean(A=OR Boolean(A=Get bIsSkydiveFloating(), B=Get bIsSkydiveDiveMode()), B=Get bIsParachuteOpened())
   Set SkydiveFloatingBin(execute=@n3.then_1, SkydiveFloatingBin:double=SelectFloat(A=0.025, B=1, bPickA=@n4)) -> (Output_Get:double)
   Set IsSkydive(execute=@n3.then_2, IsSkydive:bool=@n4) -> (Output_Get:bool)
   n5 = Branch(execute=@n3.then_3, Condition=Get IsSkydive() -> (IsSkydive:bool))
   > Set Linear Scale WO(Linear Scale WO:Vector="0.1,0.1,0.1") -> (Output_Get:Vector)
   > Set Linear Scale W(Linear Scale W:Vector="0.1,0.2,0.2") -> (Output_Get:Vector)
   > Set Linear Scale Cape(Linear Scale Cape:Vector="0.6,0.3,0.3") -> (Output_Get:Vector)
   Set Linear Scale WO(execute=@n5.else, Linear Scale WO="0.6,0.2,0.2")
   > Set Linear Scale W(Linear Scale W="0.1,0.2,0.2")
   > Set Linear Scale Cape(Linear Scale Cape="0.55,0.2,0.2")
   n6 = Get Skel Mesh()
   Set Hair Up(
    execute=@n3.then_4
    Hair Up:double=float * float(
     A=GetAlphaFromReferenceTransformDelta(SkeletalMeshComponent=@n6, BoneName=spine_05, BaseName=head, InputRange=(X=0,Y=90), TwistAxis="0,0,1")
     B=1
    ) {is=title K2Node_PromotableOperator}
   ) -> (Output_Get:double)
   n7 = Gravity Override State Function DEPRECATED(
    execute=@n3.then_0
    CharacterPartAnimInstance=Self-Reference()
    GameTimeInSeconds=GetGameTimeInSeconds()
    GravityOverrideParams=Get Gravity Override Params()
   )
   > Set GravityOverride(GravityOverride=@n7)
  ]}
]
variables [
 ar_firingPosition {type=integer}
 Body_Mesh {type=SkeletalMeshComponent}
 Gravity Override Params {
  type=GravityOverrideParamsStruct
  default {
   emoteOrMelee {gravityOverride=<0 0 -500>}
   onGround_Moving {gravityOverride=<0 0 -800>}
   onGround_Standing {gravityOverride=<0 0 -500>}
   skydive_DivingDown {gravityOverride=<-7000 2000 0>}
   skydive_FloatingDown {gravityOverride=<0 -300 200>}
   skydive_GoingUp {gravityOverride=<0 -300 200>}
  }
 }
 GravityOverride {type=Vector}
 Hair Up {type=number}
 Is BackPack {type=number}
 Is Melee {type=number}
 Is Skydive Floating {type=number}
 Is Wearing Cape {type=boolean}
 IsSkydive {type=boolean}
 isWearingBackbling {type=boolean}
 Linear Scale Cape {type=Vector}
 Linear Scale W {type=Vector, default=<0.32 0.2 0.2>}
 Linear Scale WO {type=Vector, default=<0.5 0.2 0.2>}
 Skel Mesh {type=SkeletalMeshComponent}
 skydive Alpha {type=Vector}
 SkydiveFloatingBin {type=number}
]
